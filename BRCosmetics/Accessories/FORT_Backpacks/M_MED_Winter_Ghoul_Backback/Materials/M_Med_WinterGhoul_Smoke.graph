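# Translucent lit smoke material for the Winter Ghoul backpack cosmetic, built on skinned-mesh support and an accessory tech-art override function.
# Panning noise distorts albedo and normal UVs; the color-adjusted, noise-modulated albedo drives base color and emissive.
# Scrolling blurred noise, weakened by vertex red, drives WPO, whose inverse joins two-layer noise in a vertex-red-masked opacity, depth-faded except on mobile.
# The Fresnel branch is left unconnected.
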Material /BRCosmetics/Accessories/FORT_Backpacks/M_MED_Winter_Ghoul_Backback/Materials/M_Med_WinterGhoul_Smoke
graph {
 properties {
  bEnableResponsiveAA=true
  bUseMaterialAttributes=true
  bUsedWithMeshDeformer=true
  bUsedWithSkeletalMesh=true
  blendMode=BLEND_Translucent
  opacityMaskClipValue=0.1
  pixelDepthOffsetMode=PDOM_Legacy
  usedShadingModels=Default Lit
 }
 nodes [
  Constant() {r=1}
  n1 = MaterialFunctionCall(
   Brightness=ScalarParameter() {defaultValue=1, parameterName=Fresnel-Brightness}
   Exponent=ScalarParameter() {defaultValue=2, parameterName=Fresnel-Exponent}
   Vertex / Pixel Normal=VertexNormalWS()
  ) {materialFunction=&MF_Fresnel_FX}
  n2 = ScalarParameter() {parameterName=FresnelFill}
  n3 = VertexColor()
  LinearInterpolate(
   A=Clamp(None=Add(A=@n1.Fresnel, B=@n2))
   Alpha=Multiply(A=@n3.G, B=ScalarParameter() {defaultValue=10, parameterName=FresnelPlacementMask-Hardness})
  )
  n4 = ScalarParameter() {parameterName=HueShift}
  n5 = TextureCoordinate() {coordinateIndex=1}
  n6 = ScalarParameter() {defaultValue=0.65, parameterName=NoiseTiling-WSO}
  n7 = Time()
  n8 = ScalarParameter() {defaultValue=10, parameterName=PanSpeed-WSO}
  n9 = TextureSampleParameter2D(UVs=Panner(Coordinate=Multiply(A=@n5, B=@n6), Time=Multiply(A=@n7, B=@n8)) {bFractionalPart=true, speedY=0.01}) {
   parameterName=Distort
   samplerType=SAMPLERTYPE_Masks
   texture=/Game/Gadgets/Assets/VinderTech_GliderChute/Glider_Vampire/Textures/T_Glider_Bat_Noises
  }
  n10 = ScalarParameter() {defaultValue=0.01, parameterName=Distortion_Diffuse}
  n11 = TextureSampleParameter2D(UVs=Add(A=TextureCoordinate(), B=Multiply(A=@n9.B, B=@n10))) {
   parameterName=Texture-Albedo
   texture=/BRCosmetics/Accessories/FORT_Backpacks/M_MED_Winter_Ghoul_Backback/Textures/T_M_WinterGhoul_Backpack_D
  }
  n12 = Multiply(
   A=Power(Base=@n11.RGB, Exp=ScalarParameter() {defaultValue=1, parameterName=DiffusseContrast})
   B=ScalarParameter() {defaultValue=0.75, parameterName=DiffuseBrightness}
  )
  n13 = ScalarParameter() {parameterName=DiffuseNoiseContrast_1}
  n14 = ScalarParameter() {defaultValue=1, parameterName=DiffuseNoiseContrast_2}
  n15 = VectorParameter() {parameterName=NoiseUV, defaultValue=<rgba 1 1 0 0>}
  n16 = Multiply(A=TextureCoordinate() {coordinateIndex=1}, B=AppendVector(A=@n15.R, B=@n15.R))
  n17 = Multiply(A=@n9.B, B=@n10)
  n18 = TextureSampleParameter2D(UVs=Panner(Coordinate=Add(A=@n16, B=@n17), Speed=AppendVector(A=@n15.B, B=@n15.B))) {
   parameterName=Noise
   samplerType=SAMPLERTYPE_Masks
   texture=&T_Glider_Bat_Noises
  }
  n19 = Multiply(
   A=Desaturation(
    None=MaterialFunctionCall(Hue Shift Percentage=@n4, Texture=@n12) {materialFunction=&HueShift}.Result
    Fraction=ScalarParameter() {parameterName=Saturation}
   )
   B=Clamp(None=LinearInterpolate(A=@n13, B=@n14, Alpha=@n18.G))
  )
  n20 = ScalarParameter() {defaultValue=1, parameterName=EmmissiveMultiplier}
  n21 = ScalarParameter() {defaultValue=2, parameterName=VertColorRPower}
  n22 = Panner(
   Coordinate=Multiply(A=TextureCoordinate() {coordinateIndex=1, vTiling=0.75}, B=ScalarParameter() {defaultValue=1, parameterName=NoiseTiling-A})
   Time=Multiply(A=Time(), B=ScalarParameter() {defaultValue=10, parameterName=PanSpeed-A})
  ) {bFractionalPart=true, speedX=0.01, speedY=0.01}
  n23 = Panner(
   Coordinate=Multiply(A=TextureCoordinate() {coordinateIndex=1, vTiling=0.8}, B=ScalarParameter() {defaultValue=1.2, parameterName=NoiseTiling-B})
   Time=Multiply(A=Time(), B=ScalarParameter() {defaultValue=6, parameterName=PanSpeed-B})
  ) {bFractionalPart=true, speedX=-0.01, speedY=0.075}
  n24 = LinearInterpolate(
   A=ScalarParameter() {parameterName=TailSmokeOpacity_ValueA}
   B=ScalarParameter() {defaultValue=2, parameterName=TailSmokeOpacity_ValueB}
   Alpha=Multiply(
    A=TextureSample(UVs=@n22) {samplerType=SAMPLERTYPE_Masks, texture=&T_Glider_Bat_Noises}.R
    B=TextureSample(UVs=@n23) {samplerType=SAMPLERTYPE_Masks, texture=&T_Glider_Bat_Noises}.R
   )
  )
  n25 = Clamp(None=Power(Base=Clamp(None=@n24), Exp=ScalarParameter() {defaultValue=1, parameterName=TailSmokeOpacity_Exponent3}))
  n26 = ScalarParameter() {defaultValue=1, parameterName=TailSmokeOpacity_Hardness}
  n27 = OneMinus(None=Clamp(None=Multiply(A=@n25, B=@n26)))
  n28 = VertexColor()
  n29 = ScalarParameter() {defaultValue=0.9, parameterName=WOS_MaskMultiplier}
  n30 = Panner(
   Coordinate=Multiply(A=TextureCoordinate() {coordinateIndex=1}, B=ScalarParameter() {defaultValue=1, parameterName=WSO_Tiling})
   Time=Multiply(A=Time(), B=ScalarParameter() {defaultValue=2, parameterName=WSO_Speed})
  ) {bFractionalPart=true, speedX=0.01, speedY=0.1}
  n31 = Multiply(
   A=OneMinus(None=Multiply(A=@n28.R, B=@n29))
   B=Multiply(
    A=TextureSample(UVs=@n30) {samplerType=SAMPLERTYPE_Grayscale, texture=&LowResBlurredNoise}.R
    B=VectorParameter() {parameterName=WSO-Strength, defaultValue=<rgba 3 3 4 1>}.RGB
   )
  )
  n32 = OneMinus(None=Saturate(None=@n31))
  n33 = ScalarParameter() {defaultValue=1, parameterName=Hardness}
  n34 = LinearInterpolate(A=Multiply(A=Add(A=@n27, B=@n32), B=Multiply(A=@n3.R, B=@n33)), Alpha=@n3.R)
  n35 = StaticSwitchParameter(True=Multiply(A=Power(Base=@n3.R, Exp=@n21), B=@n34) {constA=1}, False=@n34) {parameterName=UseVertColorRPow}
  n36 = DepthFade(Opacity=@n35, FadeDistance=ScalarParameter() {defaultValue=10, parameterName=DepthFade})
  n37 = ScalarParameter() {defaultValue=0.01, parameterName=Distortion_Diffuse}
  n38 = TextureSampleParameter2D(UVs=Add(A=TextureCoordinate(), B=Multiply(A=@n9.B, B=@n37))) {
   parameterName=TextureNormal
   samplerType=SAMPLERTYPE_Normal
   texture=/BRCosmetics/Accessories/FORT_Backpacks/M_MED_Winter_Ghoul_Backback/Textures/T_M_Winter_Ghoul_Backpack_N
  }
  n39 = SetMaterialAttributes!(
   MaterialAttributes
   Base Color=NamedRerouteDeclaration(None=@n19) {name=basecolor, nodeColor=<rgba 0.8047 1 0 1>}
   Emissive Color=NamedRerouteDeclaration(None=Multiply(A=@n19, B=@n20)) {name=emissive color, nodeColor=<rgba 0.2188 0 1 1>}
   Opacity=NamedRerouteDeclaration(None=ShadingPathSwitch(Default=@n36, Deferred=@n36, Mobile=@n35) {desc=Migrated from FeatureLevelSwitch}) {name=Opacity, nodeColor=<rgba 0.2891 0 1 1>}
   World Position Offset=NamedRerouteDeclaration(None=@n31) {name=wpo, nodeColor=<rgba 0.3125 0 1 1>}
   Normal=@n38.RGB
   Metallic=ScalarParameter() {parameterName=Metallic}
   Specular=ScalarParameter() {defaultValue=1, parameterName=Specular}
   Roughness=ScalarParameter() {defaultValue=0.9, parameterName=Roughness}
  ) {attributeSetTypes [
    #1
    #2
    #3
    #4
    #5
    #6
    #7
    #8
   ]}
  MaterialFunctionCall(In=@n39) {materialFunction=&MF_TechArtOverrides_accessories}
 ]
}
